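# Physics asset for the medium male SnowGlobe player body, set up for dynamic motion of three grenades and their rings.
# Each grenade forms a chain from the pelvis: a small capsule master body, a sphere grenade, then a thin box ring.
# Masters join the pelvis with free rotation, while grenade and ring joints carry tighter swing and twist limits.
# Pelvis and right thigh capsules are the only body shapes, approximating that hip area.
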PhysicsAsset /Game/Characters/Player/Male/Medium/Bodies/M_MED_SnowGlobe/Meshes/M_MED_Snow_Globe_Physics
bodies {
 dyn_grenadeRing_01 [{shapeType=Box, extentX=6, extentY=6, extentZ=1, center=<2 0 0>}]
 dyn_grenadeRing_02 [{shapeType=Box, extentX=6, extentY=6, extentZ=1, center=<2 0 0>}]
 dyn_grenadeRing_03 [{shapeType=Box, extentX=6, extentY=6, extentZ=1, center=<2 0 0>}]
 dyn_grenade_01 [{shapeType=Sphere, radius=4.3149, center=<6.54 -1 -4.3892>}]
 dyn_grenade_01_MSTR [{shapeType=Capsule, length=1, radius=1}]
 dyn_grenade_02 [{shapeType=Sphere, radius=4.6101, center=<6.0698 -1 -5.2368>}]
 dyn_grenade_02_MSTR [{shapeType=Capsule, length=1, radius=1}]
 dyn_grenade_03 [{shapeType=Sphere, radius=4.5, center=<6.3285 -1 -4.621>}]
 dyn_grenade_03_MSTR [{shapeType=Capsule, length=1, radius=1}]
 pelvis [{shapeType=Capsule, length=4.463, radius=12.5901, center=<3.5008 -0.0819 0.2297>}]
 thigh_r [{shapeType=Capsule, length=8.7797, radius=9.8503, center=<7.1817 -1.2194 -0.6888>, rotation=<rot 84.9869 -180 -180>}]
}
constraints [
 {
  bone1Name=thigh_r
  bone2Name=pelvis
  swing1LimitDegrees=45
  swing1Motion=Free
  swing2LimitDegrees=45
  swing2Motion=Free
  twistLimitDegrees=45
  twistMotion=Free
 }
 {
  bone1Name=dyn_grenade_01_MSTR
  bone2Name=pelvis
  swing1LimitDegrees=45
  swing1Motion=Free
  swing2LimitDegrees=45
  swing2Motion=Free
  twistLimitDegrees=45
  twistMotion=Free
 }
 {bone1Name=dyn_grenade_01, bone2Name=dyn_grenade_01_MSTR, swing1LimitDegrees=30, swing2LimitDegrees=30, twistLimitDegrees=10}
 {bone1Name=dyn_grenadeRing_01, bone2Name=dyn_grenade_01, swing1LimitDegrees=30, swing2LimitDegrees=30, twistLimitDegrees=15}
 {
  bone1Name=dyn_grenade_02_MSTR
  bone2Name=pelvis
  swing1LimitDegrees=45
  swing1Motion=Free
  swing2LimitDegrees=45
  swing2Motion=Free
  twistLimitDegrees=45
  twistMotion=Free
 }
 {bone1Name=dyn_grenade_02, bone2Name=dyn_grenade_02_MSTR, swing1LimitDegrees=30, swing2LimitDegrees=30, twistLimitDegrees=10}
 {bone1Name=dyn_grenadeRing_02, bone2Name=dyn_grenade_02, swing1LimitDegrees=30, swing2LimitDegrees=30, twistLimitDegrees=15}
 {
  bone1Name=dyn_grenade_03_MSTR
  bone2Name=pelvis
  swing1LimitDegrees=45
  swing1Motion=Free
  swing2LimitDegrees=45
  swing2Motion=Free
  twistLimitDegrees=45
  twistMotion=Free
 }
 {bone1Name=dyn_grenade_03, bone2Name=dyn_grenade_03_MSTR, swing1LimitDegrees=30, swing2LimitDegrees=30, twistLimitDegrees=10}
 {bone1Name=dyn_grenadeRing_03, bone2Name=dyn_grenade_03, swing1LimitDegrees=30, swing2LimitDegrees=30, twistLimitDegrees=15}
]
