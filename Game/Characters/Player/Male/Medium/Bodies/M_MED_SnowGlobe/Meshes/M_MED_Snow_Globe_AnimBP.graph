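# Animation Blueprint for the male medium Snow Globe player body, adding procedural secondary motion on top of the pose copied from the attached parent mesh.
# A Look At node aims head_base at neck_02, then three constraints drive the dyn_grenade master bones from pelvis and spine_01, weighted toward the pelvis.
# A RigidBody node then simulates the dedicated physics asset in pelvis space with scaled-down component motion.
# The event graph holds no logic.
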AnimBlueprint /Game/Characters/Player/Male/Medium/Bodies/M_MED_SnowGlobe/Meshes/M_MED_Snow_Globe_AnimBP
target_skeleton=&/M_MED_Snow_Globe_Skeleton
graphs [
 AnimGraph {nodes [
   n1 = Look At - Bone: head_base(ComponentPose=Local To Component(LocalPose=Copy Pose From Mesh() {bUseAttachedParent=true}), Alpha=1) {
    alphaBoolBlend {blendOption=Linear}
    boneToModify {boneName=head_base}
    lookAtLocation {x=0}
    lookAtTarget {boneReference {boneName=neck_02}}
    lookAt_Axis {axis=<1 0>}
   }
   n2 = "Constraint - dyn_grenade_01_MSTR (PP)"!(ComponentPose=@n1, Alpha=1, ConstraintWeights_0=0.7, ConstraintWeights_1:float=0.3) -> (Pose) {
    alphaBoolBlend {blendOption=Linear}
    boneToModify {boneName=dyn_grenade_01_MSTR}
    constraintSetup [{targetBone {boneName=pelvis}}, {targetBone {boneName=spine_01}}]
    constraintWeights [1,1]
   }
   n3 = "Constraint - dyn_grenade_02_MSTR (PP)"!(ComponentPose=@n2, Alpha=1, ConstraintWeights_0=0.7, ConstraintWeights_1:float=0.3) -> (Pose) {
    alphaBoolBlend {blendOption=Linear}
    boneToModify {boneName=dyn_grenade_02_MSTR}
    constraintSetup [{targetBone {boneName=pelvis}}, {targetBone {boneName=spine_01}}]
    constraintWeights [1,1]
   }
   n4 = "Constraint - dyn_grenade_03_MSTR (PP)"!(ComponentPose=@n3, Alpha=1, ConstraintWeights_0=0.7, ConstraintWeights_1:float=0.3) -> (Pose) {
    alphaBoolBlend {blendOption=Linear}
    boneToModify {boneName=dyn_grenade_03_MSTR}
    constraintSetup [{targetBone {boneName=pelvis}}, {targetBone {boneName=spine_01}}]
    constraintWeights [1,1]
   }
   n5 = RigidBody(ComponentPose=@n4) {
    bOverrideWorldGravity=true
    overridePhysicsAsset=&/M_MED_Snow_Globe_Physics
    simulationSpace=BaseBoneSpace
    alphaBoolBlend {blendOption=Linear}
    baseBoneRef {boneName=pelvis}
    componentAppliedLinearAccClamp=<3000 3000 3000>
    componentLinearAccScale=<0.25 0.15 0.15>
    componentLinearVelScale=<0.3 0.2 0.2>
    overrideWorldGravity {z=-980}
   }
   Output Pose(Result=Component To Local(ComponentPose=@n5)) {layerGroup=None}
  ]}
 EventGraph {nodes [
   TryGetPawnOwner()
   Event BlueprintUpdateAnimation()
  ]}
]
variables []
